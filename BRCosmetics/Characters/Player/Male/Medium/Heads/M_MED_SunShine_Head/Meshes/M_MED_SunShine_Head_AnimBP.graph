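# Animation Blueprint for the SunShine male head cosmetic, switching its material-driven mouth frame for emotes and tracking targeting state.
# Each update, and again on emote begin/end, it sets the "Smile Mouth animation Frame" scalar on the part mesh's dynamic material from bIsPlayingEmote.
# Targeting begin/end events convert bIsTargeting into isTargeting and isNotTargeting floats, which nothing shown here reads.
# The parent BlueprintUpdateAnimation call is present but unwired.
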AnimBlueprint /BRCosmetics/Characters/Player/Male/Medium/Heads/M_MED_SunShine_Head/Meshes/M_MED_SunShine_Head_AnimBP
target_skeleton=/Game/Characters/Player/Common/Fortnite_Base_Head/Fortnite_Base_Head_Export_Skeleton
graphs [EventGraph {nodes [
   Event BlueprintUpdateAnimation()
   > n1 = Get Dynamic Material Instance(SkeletalMeshComponent=GetPartSkeletalMeshComponent())
   > n2 = Branch(Condition=Get bIsPlayingEmote())
   > SetScalarParameterValue(self=@n1, ParameterName=Smile Mouth animation Frame, Value=8)
   SetScalarParameterValue(execute=@n2.else, self=@n1, ParameterName=Smile Mouth animation Frame, Value=3)
   n3 = Event OnBeginTargeting() {is=title K2Node_Event}
   n4 = "To Float (Boolean)"(InBool=Get bIsTargeting())
   Set isTargeting(execute=[@n3.then, @n5.then], isTargeting:double=@n4) -> (Output_Get:double)
   > Set isNotTargeting(isNotTargeting:double=float - float(B=@n4) {is=title K2Node_CallFunction}) -> (Output_Get:double)
   n5 = Event OnEndTargeting() {is=title K2Node_Event}
   n6 = Event OnBeginPlayingEmote() {is=title K2Node_Event}
   n7 = Get Dynamic Material Instance(execute=[@n6.then, @n8.then], SkeletalMeshComponent=GetPartSkeletalMeshComponent())
   > Is Valid(InputObject=@n7)
   > SetScalarParameterValue(self=@n7, ParameterName=Smile Mouth animation Frame, Value=SelectFloat(A=8, B=3, bPickA=Get bIsPlayingEmote()))
   n8 = Event OnEndPlayingEmote() {is=title K2Node_Event}
   TryGetPawnOwner()
   Parent: BlueprintUpdateAnimation!(execute, DeltaTimeX:float=0) -> (then) {is=title K2Node_CallParentFunction}
  ]}]
variables [
 isNotTargeting {type=number}
 isTargeting {type=number}
]
